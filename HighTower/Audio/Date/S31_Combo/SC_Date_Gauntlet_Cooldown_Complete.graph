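# Sound cue for the Date gauntlet's cooldown-complete sound, routed through the Fort_UI sound class.
# A single wave player feeds the Date_Gauntlet_Cooldown_Comp_01 wave into a root Modulator at default settings, adding slight pitch and volume variation on each play.
# The cue plays at reduced volume and references no attenuation asset.
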SoundCue /HighTower/Audio/Date/S31_Combo/SC_Date_Gauntlet_Cooldown_Complete
root=Modulator_0
nodes [
 {id=Modulator_0, children [WavePlayer_0]}
 {id=WavePlayer_0, properties {soundWaveAssetPtr=&/Date_Gauntlet_Cooldown_Comp_01}}
]
properties {maxDistance=2097152, soundClassObject=&Fort_UI, volumeMultiplier=0.5}
