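# Translucent UI material that draws the HighTower Date Combo reticle from a frame, a circle and drop lines.
# Frame and drop-line textures are offset vertically, scaled by a shared Size parameter and resolved as distance fields; the circle is a procedural SDF.
# Each element has a soft shadow mask derived from the same distance data.
# Summed masks drive a black-to-white color blend and a saturated opacity, with reticle and circle opacity parameters.
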Material /HighTower/Items/Date/Combo/UI/Materials/M_UI_DateCombo_Reticle
graph {
 properties {bAutomaticallySetUsageInEditor=false, blendMode=BLEND_Translucent, materialDomain=MD_UI, usedShadingModels=Default Lit}
 nodes [
  n1 = NamedRerouteUsage()
  n2 = NamedRerouteUsage()
  LinearInterpolate(A=Constant3Vector(), B=Constant3Vector() {constant=<rgba 1 1 1 0>}, Alpha=Add(A=Add(A=@n1, B=@n2), B=NamedRerouteUsage()))
  n3 = NamedRerouteUsage()
  n4 = ComponentMask(None=@n3) {r=true}
  n5 = Add(A=ComponentMask(None=@n3) {g=true}) {constB=0.015}
  n6 = TextureSample(UVs=MaterialFunctionCall(UVs=AppendVector(A=@n4, B=@n5), Texture Scale=ScalarParameter() {defaultValue=0.55, parameterName=Size}) {materialFunction=&ScaleUVsByCenter}.UVs) {
   samplerSource=SSM_Clamp_WorldGroupSettings
   samplerType=SAMPLERTYPE_LinearGrayscale
   texture=/HighTower/Items/Date/Combo/UI/Textures/T_UI_DateCombo_Reticle
  }
  NamedRerouteDeclaration(None=MaterialFunctionCall!(RGBA Channels=@n6.RGBA, Glow Min=Constant() {r=0.6}, Glow Max=Constant() {r=0.75}, Outline Thickness) {materialFunction=/FortUILibrary/MaterialFunctions/Basics/MF_UI_ChannelPackedDistanceField}.Fill) {name=Frame, nodeColor=<rgba 0 0 0 1>}
  NamedRerouteDeclaration(None=Multiply(A=SmoothStep(Value=@n6.R) {constMin=0.4}) {constB=0.4}) {name=FrameShadow, nodeColor=<rgba 0 0 0 1>}
  NamedRerouteDeclaration(None=MaterialFunctionCall() {materialFunction=&GetUserInterfaceUV}.Normalized UV) {name=UV, nodeColor=<rgba 0 0 0 1>}
  n7 = MaterialFunctionCall(
   UVs=NamedRerouteUsage()
   Size=ScalarParameter() {defaultValue=0.105, parameterName=CircleSize}
   OutlineThickness=ScalarParameter() {defaultValue=0.001, parameterName=CircleThickness}
  ) {materialFunction=&MF_UI_Circle}
  NamedRerouteDeclaration(None=OneMinus(None=SmoothStep(Value=@n7.SDFOutline) {constMax=0.02, constMin=0.005})) {name=Circle, nodeColor=<rgba 0 0 0 1>}
  NamedRerouteDeclaration(None=OneMinus(None=SmoothStep(Value=@n7.SDFOutline) {constMax=0.03, constMin=0.006})) {name=CircleShadow, nodeColor=<rgba 0 0 0 1>}
  n8 = NamedRerouteUsage()
  n9 = ComponentMask(None=@n8) {r=true}
  n10 = Add(A=ComponentMask(None=@n8) {g=true}) {constB=-0.41}
  n11 = TextureSample(UVs=MaterialFunctionCall(UVs=AppendVector(A=@n9, B=@n10), Texture Scale=ScalarParameter() {defaultValue=0.55, parameterName=Size}) {materialFunction=&ScaleUVsByCenter}.UVs) {samplerType=SAMPLERTYPE_LinearGrayscale, texture=/HighTower/Items/Date/Combo/UI/Textures/T_UI_DateCombo_DropLines}
  NamedRerouteDeclaration(None=MaterialFunctionCall(RGBA Channels=@n11.RGBA, Glow Min=Constant() {r=0.55}, Glow Max=Constant() {r=0.7}) {materialFunction=&MF_UI_ChannelPackedDistanceField}.Fill) {name=DropLines, nodeColor=<rgba 0 0 0 1>}
  NamedRerouteDeclaration(None=Multiply(A=SmoothStep(Value=@n11.R) {constMin=0.3}) {constB=0.4}) {name=DropLinesShadow, nodeColor=<rgba 0 0 0 1>}
  n12 = Multiply(A=Add(A=NamedRerouteUsage(), B=NamedRerouteUsage()), B=ScalarParameter() {defaultValue=1, parameterName=ReticleOpacity})
  n13 = Add(A=NamedRerouteUsage(), B=NamedRerouteUsage())
  n14 = NamedRerouteUsage()
  n15 = ScalarParameter() {defaultValue=1, parameterName=CircleOpacity}
  Saturate(None=Add(A=Add(A=@n12, B=@n13), B=Multiply(A=@n14, B=@n15)))
 ]
}
