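# Weapon reticle extension widget for the Date Combo gauntlets, showing overheat and BallLightning ability cooldown.
# PreConstruct binds to the weapon's overheat event, driving the meter material's Progress as one minus current/max.
# Animations reveal the meter and mark overheat, and a timer set at zero heat handles delayed hiding.
# Async gameplay-tag waits on the owning pawn play the cooldown animation forward, then reversed with a completion sound.
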WidgetBlueprint /HighTower/Items/Date/Combo/UI/WBP_DateCombo_Reticle
animations [
 Anim_Intro {
  binding_count=2
  length_seconds=0.35
  object=&self:Anim_Intro
  sequence {bindings [
    WBP_Date_Combo_Reticle_C_12 {
     id=#1
     possessed_class=WBP_DateCombo_Reticle_C
     tracks [
      {
       type=MovieSceneFloatTrack
       display_name=Render Opacity
       sections [{type=MovieSceneFloatSection, end_frame=6, start_frame=0, channels [None {type=MovieSceneScriptingFloatChannel, keys [[0,0],[6,1]]}]}]
      }
      {
       type=MovieScene2DTransformTrack
       sections [{
         type=MovieScene2DTransformSection
         end_frame=6
         start_frame=0
         channels [
          Scale.X {type=MovieSceneScriptingFloatChannel, keys [[0,1.25],[6,1]]}
          Scale.Y {type=MovieSceneScriptingFloatChannel, keys [[0,1.25],[6,1]]}
         ]
        }]
      }
     ]
    }
    Image_OverheatMeter {
     id=#2
     parent_id=#1
     possessed_class=Image
     tracks [{
       type=MovieSceneFloatTrack
       display_name=Render Opacity
       sections [{
         type=MovieSceneFloatSection
         end_frame=7
         start_frame=0
         channels [None {type=MovieSceneScriptingFloatChannel, keys [[-8.0003,1],[0,0],[6,0]]}]
        }]
      }]
    }
   ]}
 }
 Anim_OnOverheat {
  binding_count=2
  length_seconds=0.35
  object=&self:Anim_OnOverheat
  sequence {bindings [
    Image_ReticleCenter {
     id=#3
     possessed_class=Image
     tracks [{
       type=MovieSceneWidgetMaterialTrack
       sections [{
         type=MovieSceneParameterSection
         end_frame=7
         start_frame=0
         channels [CircleOpacity {type=MovieSceneScriptingFloatChannel, keys [[0,1],[7,0.35]]}]
        }]
      }]
    }
    Image_OverheatMeter {
     id=#4
     possessed_class=Image
     tracks [{
       type=MovieSceneFloatTrack
       display_name=Render Opacity
       sections [{type=MovieSceneFloatSection, end_frame=7, start_frame=0, channels [None {type=MovieSceneScriptingFloatChannel, keys [[0,1],[7,0.35]]}]}]
      }]
    }
   ]}
 }
 Anim_OnEnergyBombCooldown {
  binding_count=1
  length_seconds=0.35
  object=&self:Anim_OnEnergyBombCooldown
  sequence {bindings [Image_ReticleCenter {
     id=#5
     possessed_class=Image
     tracks [{
       type=MovieSceneWidgetMaterialTrack
       sections [{
         type=MovieSceneParameterSection
         end_frame=7
         start_frame=0
         channels [ReticleOpacity {type=MovieSceneScriptingFloatChannel, keys [[0,1],[7,0.3]]}]
        }]
      }]
    }]}
 }
 Anim_OverheatStart {
  binding_count=1
  length_seconds=0.35
  object=&self:Anim_OverheatStart
  sequence {bindings [Image_OverheatMeter {
     id=#6
     possessed_class=Image
     tracks [{
       type=MovieSceneFloatTrack
       display_name=Render Opacity
       sections [{
         type=MovieSceneFloatSection
         end_frame=7
         start_frame=0
         channels [None {type=MovieSceneScriptingFloatChannel, keys [[-8.0003,1],[0,0],[7.0003,1]]}]
        }]
      }]
    }]}
 }
]
graphs [EventGraph {nodes [
   n1 = HandleOverheat!() -> (OutputDelegate, then, CurrentValue:float=~, MaxValue:float=~, State:EFortWeaponOverheatState Enum=None, bIsInCooldown:bool=~) {is=title K2Node_CustomEvent}
   > n2 = Set Current Value(Current Value:float=@n1.CurrentValue) -> (Output_Get:float)
   > SetScalarParameterValue(
    self=Get Mat_Overheat() -> (Mat_Overheat:Material Instance Dynamic Object Reference)
    ParameterName=Progress
    Value=float - float(A=1, B=float / float(A=@n2, B=@n1.MaxValue))
   )
   > n3 = Branch(Condition="Equal (Float)"(A=@n2, B=@n1.MaxValue))
   > Set WasOverheated(WasOverheated:bool=true) -> (Output_Get:bool)
   > PlayAnimation(InAnimation=Get Anim_OnOverheat() -> (Anim_OnOverheat:Widget Animation Object Reference))
   n4 = Branch(execute=@n3.else, Condition="Equal (Float)"(A=Get Current Value() -> (Current Value:float)))
   > n5 = Branch(Condition=Get WasOverheated() -> (WasOverheated:bool))
   > Set WasOverheated(WasOverheated=false)
   > n6 = "Play Animation Time Range with Finished event (legacy)"!(
    execute:Exec
    Widget:User Widget Object Reference=Self-Reference()
    InAnimation:Widget Animation Object Reference=Get Anim_OnOverheat()
    StartAtTime:float=0
    EndAtTime:float=0
    NumLoopsToPlay:int=1
    PlayMode:EUMGSequencePlayMode Enum=Reverse
    PlaybackSpeed:float=1
   ) -> (then:Exec, Result:UMGSequence Player Object Reference, Finished:Exec) {is=title K2Node_PlayAnimationTimeRange}
   n7 = Set Timer by Event(
    execute=[@n6.Finished, @n5.else]
    Delegate=Create Event()
    Time=Get TimeToHideOverheat() -> (TimeToHideOverheat:float)
    bLooping=false
    bMaxOncePerFrame=true
   )
   > Set EventToHideOverheat(EventToHideOverheat:Timer Handle Structure=@n7) -> (Output_Get:Timer Handle Structure)
   Branch(execute=@n4.else, Condition="float > float"(A=Get Current Value()))
   > n8 = Branch(Condition=Get IsOverheatHidden() -> (IsOverheatHidden:bool))
   > PlayAnimation(InAnimation=Get Anim_OverheatStart() -> (Anim_OverheatStart:Widget Animation Object Reference))
   > Set IsOverheatHidden(IsOverheatHidden:bool=false) -> (Output_Get:bool)
   Branch(
    execute=@n8.else
    Condition=Is Valid Timer Handle!(Handle:Timer Handle Structure=Get EventToHideOverheat() -> (EventToHideOverheat:Timer Handle Structure)) -> (ReturnValue:bool)
   )
   > Invalidate Timer Handle!(execute, Handle:"Timer Handle Structure (by ref)"=Get EventToHideOverheat()) -> (then, ReturnValue:Timer Handle Structure)
   HideOverheatWithoutPlayerUsage() {is=title K2Node_CustomEvent}
   n9 = Get IsOverheatHidden()
   n10 = Get Current Value()
   > Branch(Condition=AND Boolean(A=NOT Boolean(A=@n9), B="Equal (Float)"(A=@n10)))
   > PlayAnimation(InAnimation=Get Anim_OverheatStart(), PlayMode=Reverse)
   > Set IsOverheatHidden(IsOverheatHidden=true)
   Event PreConstruct()
   > Set IsOverheatHidden(IsOverheatHidden=true)
   > n11 = Cast To B_HighTower_Date_Combo!(
    execute
    Object=GetEquippedWeapon!(self:Fort Weapon Reticle Extension Widget Base Object Reference) -> (ReturnValue:Fort Weapon Object Reference)
   ) -> (then, CastFailed, AsB High Tower Date Combo:B High Tower Date Combo Object Reference) {is=title K2Node_DynamicCast}
   > n12 = Set ComboGauntlets(ComboGauntlets:B High Tower Date Combo Object Reference=@n11) -> (Output_Get:B High Tower Date Combo Object Reference)
   > Bind Event to On Overheat Value Changed(self:Fort Weapon Ranged Object Reference=@n12, Delegate=Create Event()) {is=title K2Node_AddDelegate}
   n13 = Get ComboGauntlets() -> (ComboGauntlets:B High Tower Date Combo Object Reference)
   > HandleOverheat!(
    execute
    self
    CurrentValue:float=GetCurrentOverheatValue!(self:Fort Weapon Ranged Object Reference=@n13) -> (ReturnValue:float)
    MaxValue:float=GetOverheatingMaxValue!(self:Fort Weapon Ranged Object Reference=@n13) -> (ReturnValue:float)
    State:EFortWeaponOverheatState Enum=None
    bIsInCooldown:bool=false
   ) -> (then) {is=title K2Node_CallFunction}
   n14 = GetOwningPlayerPawn!(self:User Widget Object Reference) -> (ReturnValue:Pawn Object Reference)
   > n15 = WaitGameplayTagAddToActor!(
    execute:Exec
    TargetActor:Actor Object Reference=@n14
    Tag:Gameplay Tag Structure=(TagName="Granted.Cooldown.HighTower.Date.BallLightning")
    OnlyTriggerOnce:bool=false
   ) -> (then:Exec, AsyncTaskProxy:Ability Async Wait Gameplay Tag Added Object Reference, Added:Exec) {is=title K2Node_AsyncAction}
   > n16 = WaitGameplayTagRemoveFromActor!(
    execute:Exec
    TargetActor:Actor Object Reference=@n14
    Tag:Gameplay Tag Structure=(TagName="Granted.Cooldown.HighTower.Date.BallLightning")
    OnlyTriggerOnce:bool=false
   ) -> (then:Exec, AsyncTaskProxy:Ability Async Wait Gameplay Tag Removed Object Reference, Removed:Exec) {is=title K2Node_AsyncAction}
   PlayAnimation(
    execute=@n16.Removed
    InAnimation=Get Anim_OnEnergyBombCooldown() -> (Anim_OnEnergyBombCooldown:Widget Animation Object Reference)
    PlayMode=Reverse
   )
   > PlaySound2D(Sound=Get Cooldown Complete Sound() -> (Cooldown Complete Sound:Sound Base Object Reference))
   PlayAnimation(execute=@n15.Added, InAnimation=Get Anim_OnEnergyBombCooldown())
   Event Destruct()
   > Unbind Event from On Overheat Value Changed(self:Fort Weapon Ranged Object Reference=Get ComboGauntlets(), Delegate=Create Event()) {is=title K2Node_RemoveDelegate}
   Event Construct()
   > PlayAnimation(InAnimation=Get Anim_Intro() -> (Anim_Intro:Widget Animation Object Reference))
   > n17 = GetDynamicMaterial(self=Get Image_OverheatMeter() -> (Image_OverheatMeter:Image Object Reference))
   > Set Mat_Overheat(Mat_Overheat:Material Instance Dynamic Object Reference=@n17) -> (Output_Get:Material Instance Dynamic Object Reference)
   GetOwningPlayer()
  ]}]
variables [
 ComboGauntlets {type=B_HighTower_Date_Combo_C, default=null}
 Cooldown Complete Sound {type=SoundBase, default=/HighTower/Audio/Date/S31_Combo/SC_Date_Gauntlet_Cooldown_Complete}
 Current Value {type=number}
 EventToHideOverheat {type=TimerHandle}
 IsOverheatHidden {type=boolean, default=true}
 Mat_Overheat {type=MaterialInstanceDynamic}
 TimeToHideOverheat {type=number, default=2}
 WasOverheated {type=boolean}
]
widgets {
 info {
  inheritedWidgetCount=1
  parentClass=/Script/FortniteUI.FortWeaponReticleExtensionWidgetBase
  rootWidgetClass=&Overlay
  widgetCount=4
 }
 widgets [
  Overlay_14 {widgetClassPath=&Overlay, widget=Overlay_14}
  Image_ReticleCenter {
   widgetClassPath=&Image
   bIsVariable=true
   parent=Overlay_14
   widget=Image_ReticleCenter
   properties {brush {
     bIsDynamicallyLoaded=false
     drawAs=Image
     imageType=FullColor
     mirroring=NoMirror
     resourceName=None
     resourceObject=/HighTower/Items/Date/Combo/UI/Materials/M_UI_DateCombo_Reticle
     tiling=NoTile
     imageSize=<180 180>
     margin=<ltrb 0 0 0 0>
     outlineSettings {
      bUseBrushTransparency=false
      roundingType=HalfHeightRadius
      width=0
      color {colorUseRule=UseColor_Specified, specifiedColor=<rgba 0 0 0 0>}
      cornerRadii=<0 0 0 0>
     }
     tintColor {colorUseRule=UseColor_Specified, specifiedColor=<rgba 1 1 1 1>}
     uVRegion {bIsValid=false, max=<0 0>, min=<0 0>}
    }}
   slot {class=OverlaySlot, properties {horizontalAlignment=HAlign_Center, verticalAlignment=VAlign_Center}}
  }
  Image_OverheatMeter {
   widgetClassPath=&Image
   bIsVariable=true
   parent=Overlay_14
   widget=Image_OverheatMeter
   properties {
    brush {
     bIsDynamicallyLoaded=false
     drawAs=Image
     imageType=FullColor
     mirroring=NoMirror
     resourceName=None
     resourceObject=/HighTower/Items/Date/Combo/UI/Materials/MI_UI_DateCombo_Overheat
     tiling=NoTile
     imageSize=<200 192>
     margin=<ltrb 0 0 0 0>
     outlineSettings {
      bUseBrushTransparency=false
      roundingType=HalfHeightRadius
      width=0
      color {colorUseRule=UseColor_Specified, specifiedColor=<rgba 0 0 0 0>}
      cornerRadii=<0 0 0 0>
     }
     tintColor {colorUseRule=UseColor_Specified, specifiedColor=<rgba 1 1 1 1>}
     uVRegion {bIsValid=false, max=<0 0>, min=<0 0>}
    }
    renderTransform {angle=-15, scale=<1 1>, shear=<0 0>, translation=<46 23>}
   }
   slot {class=OverlaySlot}
  }
  TutorialHighlight {widgetClassPath=&TutorialHighlight}
 ]
}
